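# Unlit, two-sided Modulate test material that appears to overlay hit and crack damage by darkening what lies behind it.
# Fortnite_HitEffects takes custom primitive data for deformation center, damage radius and destruction amount, plus interpolated pre-skinned position for cracks.
# Its diffuse blends to neutral white through DepthFade, limiting the effect to near intersecting geometry.
# Its hit/shrink offset is summed with a slight normal push and camera vector.
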Material /Game/Packages/jl_tests/Translucent_hitEffects
graph {
 properties {
  blendMode=BLEND_Modulate
  pixelDepthOffsetMode=PDOM_Legacy
  shadingModel=MSM_Unlit
  translucencyPass=MTP_BeforeDOF
  twoSided=true
  usedShadingModels=Unlit
 }
 nodes [
  n1 = MaterialFunctionCall!(
   WSDeformationCenter=VectorParameter() {bUseCustomPrimitiveData=true, parameterName=WSDeformationCenter, primitiveDataIndex=20}.RGB
   Code_DestructionAmount=ScalarParameter() {bUseCustomPrimitiveData=true, group=z_code, parameterName=Code_DestructionAmount, primitiveDataIndex=27}
   Code_UseUvForDamage=StaticBool()
   Code_UseHitEffects=StaticBoolParameter() {defaultValue=true, group=z_code, parameterName=Code_UseHitEffects}
   Code_Emissive
   WSDamageRadius=ScalarParameter() {bUseCustomPrimitiveData=true, parameterName=WSDamageRadius, primitiveDataIndex=26}
   CrackDamageShader=StaticBool() {value=true}
   Diffuse=Constant() {r=1}
   ScalarModDamageCracks=Constant() {r=1}
   ScalarOffsetforDamageCracks
   ExposeLerpFunctionForInvertingTheHitDistanceCalc
   InvertDamageFalloff
   CrackWorldPosition=VertexInterpolator(VS=PreSkinnedPosition()).PS
   WorldPosition
   DistanceCullFadeAnimationOffset
  ) {materialFunction=&/Fortnite_HitEffects}
  n2 = VertexNormalWS()
  n3 = Constant() {r=0.0125}
  Add(
   A=Add(A=@n1.Hit And Shrink WPO, B=Multiply(A=@n2, B=@n3))
   B=MaterialFunctionCall!(Use World Position Offsets) {materialFunction=/Engine/Functions/Engine_MaterialFunctions02/Utility/CameraVectorWithWPOOptions}.Result
  )
  LinearInterpolate(A=@n1.Diffuse, B=Constant() {r=1}, Alpha=DepthFade() {fadeDistanceDefault=3})
 ]
}
